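# Sound cue for a Season 33 train station PA announcement, named for a Castle arrival at 90 seconds.
# A Concatenator plays a generic S36 station jingle, then one of three equally weighted female schedule voice lines.
# Base submix and submix sends are disabled, with a pre-effect send feeding the station announcement audio bus.
# Concurrency overrides cap simultaneous instances and set a retrigger time.
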SoundCue /Train/Sound/Cues/PA_Announcements/S33/SC_Train_Station_PA_S33_Castle_Arrival_90_Sec
root=Concatenator_0
nodes [
 {id=Concatenator_0, children [Mixer_0,Mixer_2], properties {inputVolume [1,1]}}
 {id=Mixer_0, children [WavePlayer_3], properties {inputVolume [1]}}
 {id=Mixer_2, children [Random_0], properties {inputVolume [1]}}
 {id=Random_0, children [WavePlayer_1,WavePlayer_2,WavePlayer_4], properties {bShouldExcludeFromBranchCulling=true, weights [1,1,1]}}
 {id=WavePlayer_1, properties {soundWaveAssetPtr=/Train/Sound/Waves/VO/S33/VO_Train_Station_PA_S33_Female_Small_Schedule_01}}
 {id=WavePlayer_2, properties {soundWaveAssetPtr=/Train/Sound/Waves/VO/S33/VO_Train_Station_PA_S33_Female_Small_Schedule_02}}
 {id=WavePlayer_3, properties {soundWaveAssetPtr=/Train/Sound/Cues/PA_Announcements/S36/MX_S36_POI_TrainStation_Jingle_Generic_STNG}}
 {id=WavePlayer_4, properties {soundWaveAssetPtr=/Train/Sound/Waves/VO/S33/VO_Train_Station_PA_S33_Female_Small_Schedule_03}}
]
properties {
 bEnableBaseSubmix=false
 bEnableSubmixSends=false
 maxDistance=2097152
 volumeMultiplier=1
 concurrencyOverrides {maxCount=2, retriggerTime=3}
 preEffectBusSends [{audioBus=/Train/Sound/AB_TrainStationAnnouncmentBus, sendLevel=0.4, soundSourceBus=null}]
}
