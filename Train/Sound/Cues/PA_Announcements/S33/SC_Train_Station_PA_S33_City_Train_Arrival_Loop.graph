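# Sound cue for the looping S33 station PA jingle that signals an arriving city train.
# A single looping wave player plays the auditory guidance wave.
# Base submix and submix sends are disabled; the audio goes to the station announcement bus through a pre-effect send.
# Concurrency overrides cap simultaneous instances and set a retrigger time.
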SoundCue /Train/Sound/Cues/PA_Announcements/S33/SC_Train_Station_PA_S33_City_Train_Arrival_Loop
root=WavePlayer_1
nodes [{
  id=WavePlayer_1
  properties {bLooping=true, soundWaveAssetPtr=/Train/Sound/Waves/Jingles/S33/S33_Train_Station_Auditory_Guidance_Train_Arriving_Loop_Small}
 }]
properties {
 bEnableBaseSubmix=false
 bEnableSubmixSends=false
 maxDistance=2097152
 volumeMultiplier=1
 concurrencyOverrides {maxCount=2, retriggerTime=3}
 preEffectBusSends [{audioBus=/Train/Sound/AB_TrainStationAnnouncmentBus, sendLevel=0.6, soundSourceBus=null}]
}
