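# Sound cue for the S33 train yard PA announcement, playing a jingle followed by a female voice line.
# A root Random picks the general announcement three times in four: one of two S36 jingles, then one of thirteen general lines.
# Otherwise it plays an S33 jingle, then one of two train greetings.
# Base submix output is disabled, with a pre-effect send to the station announcement audio bus.
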SoundCue /Train/Sound/Cues/PA_Announcements/S33/SC_Train_Station_PA_S33_Train_Yard_Announcement
root=Random_2
nodes [
 {id=Concatenator_0, children [Random_3,Mixer_0], properties {inputVolume [0.9,1]}}
 {id=Concatenator_1, children [Mixer_3,Mixer_1], properties {inputVolume [0.9,1]}}
 {id=Mixer_0, children [Random_0], properties {inputVolume [0.75]}}
 {id=Mixer_1, children [Random_1], properties {inputVolume [0.75]}}
 {id=Mixer_3, children [WavePlayer_30], properties {inputVolume [1]}}
 {
  id=Random_0
  children [
   WavePlayer_0
   WavePlayer_1
   WavePlayer_2
   WavePlayer_3
   WavePlayer_4
   WavePlayer_5
   WavePlayer_6
   WavePlayer_7
   WavePlayer_8
   WavePlayer_9
   WavePlayer_10
   WavePlayer_11
   WavePlayer_12
  ]
  properties {bShouldExcludeFromBranchCulling=true, weights [1,1,1,1,1,1,1,1,1,1,1,1,1]}
 }
 {id=Random_1, children [WavePlayer_13, WavePlayer_14], properties {weights [1,1]}}
 {
  id=Random_2
  children [Concatenator_0, Concatenator_0, Concatenator_0, Concatenator_1]
  properties {bShouldExcludeFromBranchCulling=true, weights [1,1,1,1]}
 }
 {id=Random_3, children [WavePlayer_15, WavePlayer_16], properties {bShouldExcludeFromBranchCulling=true, weights [1,1]}}
 {id=WavePlayer_0, properties {soundWaveAssetPtr=/Train/Sound/Waves/VO/S33/VO_Train_Station_PA_S33_Female_Small_General_01}}
 {id=WavePlayer_1, properties {soundWaveAssetPtr=/Train/Sound/Waves/VO/S33/VO_Train_Station_PA_S33_Female_Small_General_02}}
 {id=WavePlayer_10, properties {soundWaveAssetPtr=/Train/Sound/Waves/VO/S33/VO_Train_Station_PA_S33_Female_Small_General_11}}
 {id=WavePlayer_11, properties {soundWaveAssetPtr=/Train/Sound/Waves/VO/S33/VO_Train_Station_PA_S33_Female_Small_General_12}}
 {id=WavePlayer_12, properties {soundWaveAssetPtr=/Train/Sound/Waves/VO/S33/VO_Train_Station_PA_S33_Female_Small_General_13}}
 {id=WavePlayer_13, properties {soundWaveAssetPtr=/Train/Sound/Waves/VO/S33/VO_Train_Station_PA_S33_Female_Small_Greeting_Train}}
 {id=WavePlayer_14, properties {soundWaveAssetPtr=/Train/Sound/Waves/VO/S33/VO_Train_Station_PA_S33_Female_Small_Greeting_Train_Alt}}
 {id=WavePlayer_15, properties {soundWaveAssetPtr=/Train/Sound/Cues/PA_Announcements/S36/MX_S36_POI_TrainStation_Jingle_Generic_STNG}}
 {id=WavePlayer_16, properties {soundWaveAssetPtr=/Train/Sound/Cues/PA_Announcements/S36/MX_S36_POI_TrainStation_Jingle_Announcement_STNG}}
 {id=WavePlayer_2, properties {soundWaveAssetPtr=/Train/Sound/Waves/VO/S33/VO_Train_Station_PA_S33_Female_Small_General_03}}
 {id=WavePlayer_3, properties {soundWaveAssetPtr=/Train/Sound/Waves/VO/S33/VO_Train_Station_PA_S33_Female_Small_General_04}}
 {id=WavePlayer_30, properties {soundWaveAssetPtr=/Train/Sound/Waves/Jingles/S33/S33_Train_Station_Jingle_Small_Misc_D_01}}
 {id=WavePlayer_4, properties {soundWaveAssetPtr=/Train/Sound/Waves/VO/S33/VO_Train_Station_PA_S33_Female_Small_General_05}}
 {id=WavePlayer_5, properties {soundWaveAssetPtr=/Train/Sound/Waves/VO/S33/VO_Train_Station_PA_S33_Female_Small_General_06}}
 {id=WavePlayer_6, properties {soundWaveAssetPtr=/Train/Sound/Waves/VO/S33/VO_Train_Station_PA_S33_Female_Small_General_07}}
 {id=WavePlayer_7, properties {soundWaveAssetPtr=/Train/Sound/Waves/VO/S33/VO_Train_Station_PA_S33_Female_Small_General_08}}
 {id=WavePlayer_8, properties {soundWaveAssetPtr=/Train/Sound/Waves/VO/S33/VO_Train_Station_PA_S33_Female_Small_General_09}}
 {id=WavePlayer_9, properties {soundWaveAssetPtr=/Train/Sound/Waves/VO/S33/VO_Train_Station_PA_S33_Female_Small_General_10}}
]
properties {
 bEnableBaseSubmix=false
 bEnableSubmixSends=false
 maxDistance=2097152
 volumeMultiplier=1
 concurrencyOverrides {maxCount=3}
 preEffectBusSends [{audioBus=/Train/Sound/AB_TrainStationAnnouncmentBus, sendLevel=0.4, soundSourceBus=null}]
}
