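# Sound cue for the S33 train yard station PA announcement of a train departure.
# A mixer layers the small S33 departing guidance jingle at double volume with an S36 train station announcement stinger delayed three seconds.
# Output skips the base submix and submix sends, and feeds the station announcement audio bus through a pre-effect send.
# Concurrency overrides limit simultaneous instances and retriggering.
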SoundCue /Train/Sound/Cues/PA_Announcements/S33/SC_Train_Station_PA_S33_Train_Yard_Train_Departure
root=Mixer_2
nodes [
 {id=Delay_0, children [WavePlayer_1], properties {delayMax=3, delayMin=3}}
 {id=Mixer_2, children [Delay_0,WavePlayer_0], properties {inputVolume [1,2]}}
 {id=WavePlayer_0, properties {soundWaveAssetPtr=/Train/Sound/Waves/Jingles/S33/S33_Train_Station_Auditory_Guidance_Train_Departing_Small}}
 {id=WavePlayer_1, properties {soundWaveAssetPtr=/Train/Sound/Cues/PA_Announcements/S36/MX_S36_POI_TrainStation_Jingle_Announcement_STNG}}
]
properties {
 bEnableBaseSubmix=false
 bEnableSubmixSends=false
 maxDistance=2097152
 volumeMultiplier=1
 concurrencyOverrides {maxCount=2, retriggerTime=3}
 preEffectBusSends [{audioBus=/Train/Sound/AB_TrainStationAnnouncmentBus, sendLevel=0.4, soundSourceBus=null}]
}
